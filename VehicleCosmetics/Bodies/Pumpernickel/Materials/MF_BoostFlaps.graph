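# Material function for the Pumpernickel vehicle body that flutters boost flaps through World Position Offset on the incoming material attributes.
# Pivot Painter per-object data supplies each flap's pivot and axis for RotateAboutAxis, masked by vertex alpha.
# A time-driven sine-cosine wave, phase-shifted by UV3 green, sets the angle, scaled by BoostGlowIntensity from the Blueprint-only parameter group.
# The BoostFlaps static switch outputs zero offset when off.
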MaterialFunction /VehicleCosmetics/Bodies/Pumpernickel/Materials/MF_BoostFlaps
graph {nodes [
  n1 = MaterialFunctionCall() {materialFunction=/Engine/Functions/Engine_MaterialFunctions02/PivotPainter/PivotPainter_PerObjectData}
  n2 = CrossProduct(A=VectorParameter() {defaultValue=<rgba 1 0 0 0>}.RGB, B=@n1.X-Axis Vector)
  n3 = TextureCoordinate() {coordinateIndex=3}
  n4 = Add(
   A=Multiply(A=ComponentMask(None=@n3) {g=true}, B=ScalarParameter() {defaultValue=0.5, group=BoostFlaps, parameterName=Randomness})
   B=Multiply(A=Time(), B=ScalarParameter() {defaultValue=2, group=BoostFlaps, parameterName=Speed})
  )
  n5 = Sine(None=@n4) {period=0.5}
  n6 = Cosine(None=@n4)
  n7 = Multiply(A=ConstantBiasScale(None=Multiply(A=@n5, B=@n6)), B=ScalarParameter() {defaultValue=0.05, group=BoostFlaps, parameterName=GlobalScale})
  n8 = ScalarParameter() {group=BP-USE-ONLY, parameterName=BoostGlowIntensity}
  n9 = Multiply(
   A=RotateAboutAxis(
    NormalizedRotationAxis=Normalize(VectorInput=@n2)
    RotationAngle=Multiply(A=@n7, B=@n8)
    PivotPoint=@n1.Pivot Position
    Position=WorldPosition().XYZ
   )
   B=Ceil(None=VertexColor().A)
  )
  n10 = Constant()
  FunctionOutput(None=SetMaterialAttributes!(
    MaterialAttributes=FunctionInput() {inputType=FunctionInput_MaterialAttributes}
    WorldPositionOffset=StaticSwitchParameter(True=@n9, False=@n10) {defaultValue=true, group=BoostFlaps, parameterName=BoostFlaps}
   ) {attributeSetTypes [#1]})
 ]}
