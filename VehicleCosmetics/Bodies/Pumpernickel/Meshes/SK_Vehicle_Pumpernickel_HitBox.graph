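# Physics asset providing the hitbox for the Pumpernickel vehicle body cosmetic's skeletal mesh.
# It holds a single body on the root bone, made of one box shape offset forward along X and raised above the origin.
# There are no constraints, so the whole vehicle is represented by one simple box rather than per-bone collision.
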PhysicsAsset /VehicleCosmetics/Bodies/Pumpernickel/Meshes/SK_Vehicle_Pumpernickel_HitBox
bodies {root [{shapeType=Box, extentX=412.2233, extentY=270.9452, extentZ=94.0621, shapeName=root_box, center=<57.6549 0 77.4029>}]}
constraints []
